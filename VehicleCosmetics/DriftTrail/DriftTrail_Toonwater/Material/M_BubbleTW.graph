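# Translucent unlit Niagara sprite material that draws a procedural, wobbling bubble for the Toonwater drift trail.
# A panning flow texture distorts centered UVs, whose squared radius forms a sharp disc that fades toward its middle.
# This mask drives warm-tinted emissive, multiplied by particle color and passed through MF_FX_EmissiveAdjustments, and combines with particle alpha.
# Dynamic parameters control distortion and opacity, but "speed" scales noise UVs while "tile" scales time.
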Material /VehicleCosmetics/DriftTrail/DriftTrail_Toonwater/Material/M_BubbleTW
graph {
 properties {
  bAutomaticallySetUsageInEditor=false
  bUsedWithNiagaraSprites=true
  blendMode=BLEND_Translucent
  refractionMethod=RM_2DOffset
  shadingModel=MSM_Unlit
  usedShadingModels=Unlit
 }
 nodes [
  n1 = ParticleColor()
  n2 = TextureCoordinate()
  n3 = DynamicParameter() {paramNames [intensity,speed,tile,Opacity]}
  n4 = Time()
  n5 = TextureSample(UVs=Panner(Coordinate=Multiply(A=@n2, B=@n3.speed), Time=Multiply(A=@n3.tile, B=@n4)) {speedX=0.1, speedY=0.3}) {texture=/DealershipSharedLibrary/FX/Textures/Noise/T_VCVFX_Flow02_N}
  n6 = Add(
   A=Add(A=TextureCoordinate() {uTiling=2, vTiling=2}, B=Constant() {r=-1})
   B=Multiply(A=ComponentMask(None=@n5.RGB) {g=true, r=true}, B=@n3.intensity)
  )
  n7 = DotProduct(A=@n6, B=@n6)
  n8 = OneMinus(None=@n7)
  n9 = Constant() {r=0.8}
  MaterialExpressionConstantClamp_19 = Saturate(None=Multiply(A=Subtract(A=@n8, B=@n9), B=Constant() {r=32}))
  MaterialExpressionConstantClamp_20 = Saturate(None=Multiply(A=@n7, B=@n3.Opacity))
  n10 = Multiply(A=@MaterialExpressionConstantClamp_19, B=@MaterialExpressionConstantClamp_20)
  MaterialExpressionConstantClamp_22 = Saturate(None=Multiply(A=@n1.A, B=@n10))
  DeriveNormalZ(InXY=@n6)
  MaterialExpressionConstantClamp_21 = Saturate(None=Multiply(A=Constant() {r=10}, B=@n10))
  n11 = Multiply(A=Constant3Vector() {constant=<rgba 0.2961 0.2874 0.1746 1>}, B=@MaterialExpressionConstantClamp_21)
  n12 = Multiply(A=Constant() {r=0.5}, B=@n10)
  MaterialFunctionCall(Emissive=Multiply(A=Add(A=@n11, B=@n12), B=@n1.RGB)) {materialFunction=&MF_FX_EmissiveAdjustments}
 ]
}
