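# Translucent unlit material for Niagara ribbons that draws a toon-water drift trail.
# A scrolling shape texture distorts UVs into a wave edge, forming a bright border, tinted body and mask for two panning, combined caustic layers.
# Particle color, optionally desaturated via IsPainted?, and dynamic parameters drive color and brightness, while UV gradients, particle alpha and depth fade shape opacity.
# Quality switches drop noisy border glow on Low and add caustic UV distortion on Epic.
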Material /VehicleCosmetics/DriftTrail/DriftTrail_Toonwater/Material/M_TWater
graph {
 properties {
  bAutomaticallySetUsageInEditor=false
  bUsedWithNiagaraRibbons=true
  blendMode=BLEND_Translucent
  refractionMethod=RM_2DOffset
  shadingModel=MSM_Unlit
  usedShadingModels=Unlit
 }
 nodes [
  n1 = DynamicParameter() {parameterIndex=3, paramNames [IsPainted?, BorderPowerColor, PlaneColorPower, NoiseInnerColor]}
  n2 = ParticleColor()
  n3 = ScalarParameter() {parameterName=Saturation}
  n4 = Multiply(A=Constant() {r=0.5}, B=@n3)
  n5 = ParticleColor()
  n6 = LinearInterpolate(A=@n3, B=@n4, Alpha=@n5.R)
  n7 = LinearInterpolate(A=@n4, B=@n3, Alpha=Add(A=@n5.G, B=@n5.G))
  n8 = StaticSwitchParameter(True=Saturate(None=Add(A=@n6, B=@n7)), False=@n3) {defaultValue=true, parameterName=UseSaturationOffset}
  n9 = If(
   A=@n1.IsPainted?
   "A > B"=@n2.RGB
   "A == B"=@n2.RGB
   "A < B"=Desaturation(None=@n2.RGB, Fraction=If(A=@n1.IsPainted?, "A > B"=@n2.RGB, "A == B"=@n2.RGB, "A < B"=@n8) {constB=0.5}) {luminanceFactors=<rgba 0.3 0.59 0.11 0>}
  ) {constB=0.5}
  n10 = Multiply(A=@n9, B=@n1.BorderPowerColor)
  n11 = ScalarParameter() {defaultValue=2, parameterName=WaveSharpness}
  n12 = Rotator(
   Coordinate=Multiply(A=TextureCoordinate() {uTiling=0.2, vTiling=0.5}, B=DynamicParameter() {paramNames [Hardness,TileAmt,AddMultiply,Brightness]}.TileAmt)
   Time=ScalarParameter() {parameterName=Rotate}
  ) {speed=3.14}
  n13 = Constant2Vector() {r=-1}
  n14 = AppendVector(A=Constant() {r=2.5}, B=ScalarParameter() {defaultValue=1, parameterName=WaveSpeed})
  n15 = Time() {bOverride_Period=true, period=0.4}
  n16 = AppendVector(A=@n15, B=@n15)
  n17 = TextureSample(UVs=Add(A=Multiply(A=@n12, B=@n13), B=Multiply(A=@n14, B=@n16))) {texture=/DealershipSharedLibrary/FX/Textures/TexturePackage/T_VCVFX_Shapes_01_Pack_RGB}
  n18 = Add(A=@n17.G, B=ConstantBiasScale(None=@n17.B) {bias=-0.05, scale=0.1})
  n19 = TextureCoordinate()
  n20 = DynamicParameter() {parameterIndex=1, paramNames [WaveAmpli, WaveIntesity, "", ""]}
  n21 = Multiply(A=OneMinus(None=ComponentMask(None=@n19) {r=true}), B=@n20.WaveIntesity)
  n22 = Power(Base=TextureCoordinate(), Exp=@n20.WaveAmpli) {constExponent=1}
  n23 = Constant2Vector() {g=-0.7}
  n24 = ComponentMask(None=Add(A=Multiply(A=@n18, B=@n21), B=Add(A=@n22, B=@n23))) {g=true}
  n25 = Saturate(None=@n24)
  n26 = Multiply(A=@n25, B=Power(Base=OneMinus(None=@n24), Exp=ScalarParameter() {defaultValue=2, parameterName=Power}))
  n27 = Saturate(None=ConstantBiasScale(None=ComponentMask(None=@n19) {g=true}) {bias=-0.7, scale=2})
  n28 = Saturate(None=Add(A=Multiply(A=@n11, B=@n26), B=Multiply(A=@n24, B=@n27)))
  n29 = ConstantBiasScale(None=Saturate(None=OneMinus(None=@n28))) {bias=-0.98, scale=60}
  n30 = OneMinus(None=Saturate(None=ConstantBiasScale(None=@n29) {bias=-0.98, scale=5}))
  n31 = Saturate(None=ConstantBiasScale(None=@n25) {bias=-0.001, scale=5})
  n32 = OneMinus(None=@n31)
  n33 = Multiply(A=Multiply(A=Saturate(None=@n30), B=Saturate(None=@n32)), B=Constant() {r=1.5})
  n34 = Constant() {r=600}
  n35 = Saturate(None=@n29)
  n36 = Constant() {r=20}
  n37 = Constant() {r=0.5}
  n38 = Panner(Coordinate=Panner(Coordinate=TextureCoordinate() {vTiling=0.125}, Time=@n37) {speedY=0.5}) {speedX=-1, speedY=0.5}
  n39 = Panner(Coordinate=Panner(Coordinate=TextureCoordinate() {uTiling=1.25, vTiling=-0.125}, Time=@n37) {speedY=-0.5}) {speedX=-1.25, speedY=-0.35}
  n40 = ComponentMask(None=Add(
    A=TextureSample(UVs=@n38) {texture=/DealershipSharedLibrary/FX/Textures/Noise/T_VCVFX_DistortNoise02_RGB}.RGB
    B=TextureSample(UVs=@n39) {texture=&T_VCVFX_DistortNoise02_RGB}.RGB
   )) {g=true, r=true}
  n41 = Desaturation(None=@n40) {luminanceFactors=<rgba 0.3 0.59 0.11 0>}
  n42 = QualitySwitch(Default=Multiply(A=Multiply(A=@n28, B=@n35), B=Multiply(A=@n36, B=@n41)), Low=Constant())
  n43 = QualitySwitch(Default=Add(A=@n33, B=Multiply(A=@n34, B=@n42)), Low=@n33)
  n44 = Multiply(A=Multiply(A=@n2.RGB, B=@n1.PlaneColorPower), B=Multiply(A=@n31, B=Constant() {r=0.25}))
  n45 = TextureCoordinate() {uTiling=4}
  n46 = Constant() {r=0.5}
  n47 = Multiply(A=Constant() {r=0.15}, B=@n40)
  n48 = QualitySwitch(Default=@n45, Low=@n45, Epic=Add(A=Panner(Coordinate=@n45, Time=@n46) {speedY=0.5}, B=@n47))
  n49 = Multiply(A=Time(), B=ScalarParameter() {defaultValue=2, parameterName=PanSpeed})
  n50 = StaticComponentMaskParameter(None=TextureSample(UVs=Panner(Coordinate=@n48, Time=@n49) {speedX=-1.5, speedY=0.25}) {texture=/DealershipSharedLibrary/FX/Textures/Noise/T_VCVFX_CausticsMask_RGB}.RGB) {defaultG=true, parameterName=MaskRGB_01}
  n51 = TextureCoordinate() {uTiling=3, vTiling=-1}
  n52 = QualitySwitch(Default=@n51, Low=@n51, Epic=Add(A=@n47, B=Panner(Coordinate=@n51, Time=@n46) {speedY=-0.5}))
  n53 = StaticComponentMaskParameter(None=TextureSample(UVs=Panner(Coordinate=@n52, Time=@n49) {speedX=-0.75, speedY=-0.4}) {texture=&T_VCVFX_CausticsMask_RGB}.RGB) {defaultB=true, parameterName=MaskRGB_02}
  n54 = DynamicParameter() {paramNames [Hardness,TileAmt,AddMultiply,Brightness]}
  n55 = LinearInterpolate(A=Multiply(A=Multiply(A=@n50, B=@n53), B=@n54.Hardness), B=Multiply(A=@n54.Hardness, B=Add(A=@n50, B=@n53)), Alpha=@n54.AddMultiply)
  n56 = Multiply(A=Saturate(None=Multiply(A=@n31, B=@n55)), B=@n54.Brightness)
  n57 = Multiply(A=Multiply(A=@n9, B=@n1.NoiseInnerColor), B=@n56)
  MaterialFunctionCall(Emissive=Add(A=Multiply(A=@n10, B=@n43), B=Add(A=@n44, B=@n57))) {materialFunction=&MF_FX_EmissiveAdjustments}
  n58 = TextureCoordinate()
  n59 = ComponentMask(None=@n58) {g=true}
  n60 = Constant() {r=0.85}
  n61 = Constant() {r=60}
  n62 = ComponentMask(None=TextureCoordinate()) {r=true}
  n63 = Multiply(A=Power(Base=OneMinus(None=@n62), Exp=Constant() {r=2}), B=@n25)
  n64 = ComponentMask(None=TextureCoordinate()) {r=true}
  n65 = Saturate(None=ConstantBiasScale(None=@n64) {bias=-0.001, scale=10})
  n66 = Saturate(None=Multiply(A=Multiply(A=@n61, B=@n63), B=@n65))
  n67 = Multiply(A=Saturate(None=Multiply(A=@n60, B=@n66)), B=Power(Base=OneMinus(None=@n59)) {constExponent=0.3})
  n68 = ComponentMask(None=@n58) {r=true}
  Multiply(
   A=MaterialFunctionCall(Opacity In=Constant() {r=1}) {materialFunction=&MF_VCVFX_DepthFade}.Result
   B=Multiply(A=Multiply(A=@n2.A, B=@n67), B=Power(Base=@n68) {constExponent=0.5})
  )
  ScalarParameter() {defaultValue=1, parameterName=WaveIntensity}
  n69 = Saturate(None=OneMinus(None=@n64))
  StaticSwitchParameter(True=Multiply(A=@n56, B=Multiply(A=@n65, B=@n69)), False=Constant()) {parameterName=Distortion}
  Transform(None=VectorParameter() {parameterName=Position, defaultValue=<rgba 0 0 0 1>}.RGB) {transformSourceType=TRANSFORMSOURCE_Local}
 ]
}
