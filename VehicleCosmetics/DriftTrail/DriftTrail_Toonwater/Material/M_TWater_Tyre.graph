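# Translucent unlit toon water material for tyre drift trails on Niagara mesh particles.
# Two wave layers distort UV gradients with scrolling shape textures driven by wheel vertical velocity, forming borders and fills with panning caustics, noise-distorted at Epic quality.
# Particle colour tints both, desaturated when IsPainted is low, with dynamic parameters controlling intensity and tiling.
# Summed layers feed an emissive adjustment function, while both wave masks multiply with a rounded sine mask for opacity.
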Material /VehicleCosmetics/DriftTrail/DriftTrail_Toonwater/Material/M_TWater_Tyre
graph {
 properties {
  bAutomaticallySetUsageInEditor=false
  bUsedWithNiagaraMeshParticles=true
  blendMode=BLEND_Translucent
  shadingModel=MSM_Unlit
  usedShadingModels=Unlit
 }
 nodes [
  n1 = DynamicParameter() {parameterIndex=3, paramNames [IsPainted?, BorderPowerColor, PlaneColorPower, NoiseInnerColor]}
  n2 = ParticleColor()
  n3 = ScalarParameter() {parameterName=Saturation}
  n4 = Multiply(A=Constant() {r=0.5}, B=@n3)
  n5 = ParticleColor()
  n6 = Add(A=@n5.G, B=@n5.G)
  n7 = Saturate(None=Add(A=LinearInterpolate(A=@n3, B=@n4, Alpha=@n5.R), B=LinearInterpolate(A=@n4, B=@n3, Alpha=@n6)))
  n8 = If(
   A=@n1.IsPainted?
   "A > B"=@n2.RGB
   "A == B"=@n2.RGB
   "A < B"=Desaturation(None=@n2.RGB, Fraction=StaticSwitchParameter(True=@n7, False=@n3) {defaultValue=true, parameterName=UseSaturationOffset})
  ) {constB=0.5}
  n9 = ScalarParameter() {defaultValue=1, parameterName=WaveSharpness}
  n10 = TextureCoordinate()
  n11 = Rotator(
   Coordinate=Multiply(A=@n10, B=DynamicParameter() {paramNames [Hardness,TileAmt,AddMultiply,Brightness]}.TileAmt)
   Time=ScalarParameter() {parameterName=Rotate}
  ) {speed=3.14}
  n12 = Constant2Vector() {r=-1}
  n13 = DynamicParameter() {parameterIndex=1, paramNames [VelocityWheelVertical, Param2, Param3, Param4]}
  n14 = AppendVector(A=@n13.VelocityWheelVertical, B=ScalarParameter() {parameterName=WaveSpeed})
  n15 = Time()
  n16 = AppendVector(A=@n15, B=@n15)
  n17 = TextureSample(UVs=Add(A=Multiply(A=@n11, B=@n12), B=Multiply(A=@n14, B=@n16))) {texture=/DealershipSharedLibrary/FX/Textures/TexturePackage/T_VCVFX_Shapes_01_Pack_RGB}
  n18 = Add(A=@n17.G, B=Saturate(None=ConstantBiasScale(None=@n17.B) {bias=-0.05, scale=0.2}))
  n19 = TextureCoordinate()
  n20 = ComponentMask(None=@n19) {r=true}
  n21 = Multiply(A=Power(Base=OneMinus(None=@n20)) {constExponent=0}, B=ScalarParameter() {defaultValue=1, parameterName=WaveIntensity})
  n22 = Add(
   A=TextureCoordinate() {uTiling=2, vTiling=2}
   B=AppendVector(A=ScalarParameter() {parameterName=X_1AmpliWave}, B=ScalarParameter() {defaultValue=-0.05, parameterName=Y_1AmpliWave})
  )
  n23 = ComponentMask(None=Add(A=Multiply(A=@n18, B=@n21), B=OneMinus(None=@n22))) {g=true}
  n24 = Saturate(None=@n23)
  n25 = Multiply(A=@n24, B=Power(Base=OneMinus(None=@n23), Exp=ScalarParameter() {parameterName=Power}))
  n26 = Saturate(None=ConstantBiasScale(None=ComponentMask(None=@n19) {g=true}) {bias=0, scale=2})
  n27 = Saturate(None=Add(A=Multiply(A=@n9, B=@n25), B=Multiply(A=@n23, B=@n26)))
  n28 = ConstantBiasScale(None=Saturate(None=OneMinus(None=@n27))) {bias=-0.98, scale=10}
  n29 = OneMinus(None=Saturate(None=ConstantBiasScale(None=@n28) {bias=-0.98, scale=5}))
  n30 = Saturate(None=ConstantBiasScale(None=@n24) {bias=-0.001, scale=2})
  n31 = OneMinus(None=@n30)
  n32 = Multiply(A=Multiply(A=Saturate(None=@n29), B=Saturate(None=@n31)), B=Constant() {r=1.5})
  n33 = Multiply(A=@n27, B=Saturate(None=@n28))
  n34 = Multiply(A=Constant(), B=QualitySwitch(Default=Multiply(A=@n33), Low=Constant(), Medium=@n33))
  n35 = Multiply(A=Multiply(A=@n1.BorderPowerColor, B=@n8), B=QualitySwitch(Default=Add(A=@n32, B=@n34), Low=@n32))
  n36 = Constant() {r=0.25}
  n37 = TextureCoordinate() {uTiling=4}
  n38 = Constant() {r=0.5}
  n39 = Constant() {r=0.5}
  n40 = Panner(Coordinate=Panner(Coordinate=TextureCoordinate() {vTiling=0.125}, Time=@n39) {speedY=0.5}) {speedX=-1, speedY=0.5}
  n41 = Panner(Coordinate=Panner(Coordinate=TextureCoordinate() {uTiling=1.25, vTiling=-0.125}, Time=@n39) {speedY=-0.5}) {speedX=-1.25, speedY=-0.35}
  n42 = ComponentMask(None=Add(
    A=TextureSample(UVs=@n40) {texture=/DealershipSharedLibrary/FX/Textures/Noise/T_VCVFX_DistortNoise02_RGB}.RGB
    B=TextureSample(UVs=@n41) {texture=&T_VCVFX_DistortNoise02_RGB}.RGB
   )) {g=true, r=true}
  n43 = Multiply(A=Constant() {r=0.15}, B=Multiply(A=Constant() {r=20}, B=Desaturation(None=@n42)))
  n44 = QualitySwitch(Default=@n37, Low=@n37, Epic=Add(A=Panner(Coordinate=@n37, Time=@n38) {speedY=0.5}, B=@n43))
  n45 = Multiply(A=Time(), B=ScalarParameter() {defaultValue=2, parameterName=PanSpeed})
  n46 = StaticComponentMaskParameter(None=TextureSample(UVs=Panner(Coordinate=@n44, Time=@n45) {speedX=-1.5, speedY=0.25}) {texture=/DealershipSharedLibrary/FX/Textures/Noise/T_VCVFX_CausticsMask_RGB}.RGB) {defaultG=true, parameterName=MaskRGB_01}
  n47 = TextureCoordinate() {uTiling=3, vTiling=-1}
  n48 = QualitySwitch(Default=@n47, Low=@n47, Epic=Add(A=Panner(Coordinate=@n47, Time=@n38) {speedY=-0.5}, B=@n43))
  n49 = StaticComponentMaskParameter(None=TextureSample(UVs=Panner(Coordinate=@n48, Time=@n45) {speedX=-0.75, speedY=-0.4}) {texture=&T_VCVFX_CausticsMask_RGB}.RGB) {defaultB=true, parameterName=MaskRGB_02}
  n50 = DynamicParameter() {paramNames [Hardness,TileAmt,AddMultiply,Brightness]}
  n51 = Multiply(A=Multiply(A=@n46, B=@n49), B=@n50.Hardness)
  n52 = Multiply(A=@n50.Hardness, B=Add(A=@n46, B=@n49))
  n53 = Saturate(None=Multiply(A=@n30, B=LinearInterpolate(A=@n51, B=@n52, Alpha=@n50.AddMultiply)))
  n54 = Add(
   A=Multiply(A=Multiply(A=@n2.RGB, B=@n1.PlaneColorPower), B=Multiply(A=@n30, B=@n36))
   B=Multiply(A=Multiply(A=@n8, B=@n1.NoiseInnerColor), B=Multiply(A=@n53, B=@n50.Brightness))
  )
  n55 = DynamicParameter() {parameterIndex=3, paramNames [IsPainted?, BorderPowerColor, PlaneColorPower, NoiseInnerColor]}
  n56 = ParticleColor()
  n57 = ScalarParameter() {parameterName=Saturation}
  n58 = Multiply(A=Constant() {r=0.5}, B=@n57)
  n59 = ParticleColor()
  n60 = Add(A=@n59.G, B=@n59.G)
  n61 = Saturate(None=Add(A=LinearInterpolate(A=@n57, B=@n58, Alpha=@n59.R), B=LinearInterpolate(A=@n58, B=@n57, Alpha=@n60)))
  n62 = If(
   A=@n55.IsPainted?
   "A > B"=@n56.RGB
   "A == B"=@n56.RGB
   "A < B"=Desaturation(None=@n56.RGB, Fraction=StaticSwitchParameter(True=@n61, False=@n57) {defaultValue=true, parameterName=UseSaturationOffset})
  ) {constB=0.5}
  n63 = ScalarParameter() {defaultValue=1, parameterName=WaveSharpness}
  n64 = Rotator(
   Coordinate=Multiply(A=@n10, B=DynamicParameter() {paramNames [Hardness,TileAmt,AddMultiply,Brightness]}.TileAmt)
   Time=ScalarParameter() {parameterName=Rotate}
  ) {speed=3.14}
  n65 = AppendVector(A=@n13.VelocityWheelVertical, B=ScalarParameter() {parameterName=WaveSpeed})
  n66 = Time()
  n67 = AppendVector(A=@n66, B=@n66)
  n68 = TextureSample(UVs=Add(A=Multiply(A=@n64, B=@n12), B=Multiply(A=@n65, B=@n67))) {texture=&T_VCVFX_Shapes_01_Pack_RGB}
  n69 = Add(A=@n68.G, B=Saturate(None=ConstantBiasScale(None=@n68.B) {bias=-0.05, scale=0.2}))
  n70 = TextureCoordinate() {uTiling=0.5}
  n71 = Multiply(A=OneMinus(None=ComponentMask(None=@n70) {r=true}), B=ScalarParameter() {defaultValue=1, parameterName=WaveIntensity})
  n72 = TextureCoordinate() {uTiling=3, vTiling=3}
  n73 = AppendVector(A=ScalarParameter() {parameterName=X_2AmpliWave}, B=ScalarParameter() {defaultValue=-0.9, parameterName=Y_2AmpliWave})
  n74 = ComponentMask(None=Add(A=Multiply(A=@n69, B=@n71), B=Add(A=@n72, B=@n73))) {g=true}
  n75 = Saturate(None=@n74)
  n76 = Multiply(A=@n75, B=Power(Base=OneMinus(None=@n74), Exp=ScalarParameter() {parameterName=Power}))
  n77 = Saturate(None=ConstantBiasScale(None=ComponentMask(None=@n70) {g=true}) {bias=0, scale=2})
  n78 = Saturate(None=Add(A=Multiply(A=@n63, B=@n76), B=Multiply(A=@n74, B=@n77)))
  n79 = ConstantBiasScale(None=Saturate(None=OneMinus(None=@n78))) {bias=-0.98, scale=60}
  n80 = OneMinus(None=Saturate(None=ConstantBiasScale(None=@n79) {bias=-0.98, scale=5}))
  n81 = Saturate(None=ConstantBiasScale(None=@n75) {bias=-0.001, scale=2})
  n82 = OneMinus(None=@n81)
  n83 = Multiply(A=Multiply(A=Saturate(None=@n80), B=Saturate(None=@n82)), B=Constant() {r=1})
  n84 = Multiply(A=@n78, B=Saturate(None=@n79))
  n85 = Multiply(A=Constant() {r=600}, B=QualitySwitch(Default=Multiply(A=@n84), Low=Constant(), Medium=@n84))
  n86 = Multiply(A=Multiply(A=@n62, B=@n55.BorderPowerColor), B=QualitySwitch(Default=Add(A=@n83, B=@n85), Low=@n83))
  n87 = Constant() {r=0.25}
  n88 = TextureCoordinate() {uTiling=4}
  n89 = Constant() {r=0.5}
  n90 = Constant() {r=0.5}
  n91 = Panner(Coordinate=Panner(Coordinate=TextureCoordinate() {vTiling=0.125}, Time=@n90) {speedY=0.5}) {speedX=-1, speedY=0.5}
  n92 = Panner(Coordinate=Panner(Coordinate=TextureCoordinate() {uTiling=1.25, vTiling=-0.125}, Time=@n90) {speedY=-0.5}) {speedX=-1.25, speedY=-0.35}
  n93 = ComponentMask(None=Add(
    A=TextureSample(UVs=@n91) {texture=&T_VCVFX_DistortNoise02_RGB}.RGB
    B=TextureSample(UVs=@n92) {texture=&T_VCVFX_DistortNoise02_RGB}.RGB
   )) {g=true, r=true}
  n94 = Multiply(A=Constant() {r=0.15}, B=Multiply(A=Constant() {r=20}, B=Desaturation(None=@n93)))
  n95 = QualitySwitch(Default=@n88, Low=@n88, Epic=Add(A=Panner(Coordinate=@n88, Time=@n89) {speedY=0.5}, B=@n94))
  n96 = Multiply(A=Time(), B=ScalarParameter() {defaultValue=2, parameterName=PanSpeed})
  n97 = StaticComponentMaskParameter(None=TextureSample(UVs=Panner(Coordinate=@n95, Time=@n96) {speedX=-1.5, speedY=0.25}) {texture=&T_VCVFX_CausticsMask_RGB}.RGB) {defaultG=true, parameterName=MaskRGB_01}
  n98 = TextureCoordinate() {uTiling=3, vTiling=-1}
  n99 = QualitySwitch(Default=@n98, Low=@n98, Epic=Add(A=Panner(Coordinate=@n98, Time=@n89) {speedY=-0.5}, B=@n94))
  n100 = StaticComponentMaskParameter(None=TextureSample(UVs=Panner(Coordinate=@n99, Time=@n96) {speedX=-0.75, speedY=-0.4}) {texture=&T_VCVFX_CausticsMask_RGB}.RGB) {defaultB=true, parameterName=MaskRGB_02}
  n101 = DynamicParameter() {paramNames [Hardness,TileAmt,AddMultiply,Brightness]}
  n102 = Multiply(A=Multiply(A=@n97, B=@n100), B=@n101.Hardness)
  n103 = Multiply(A=@n101.Hardness, B=Add(A=@n97, B=@n100))
  n104 = Saturate(None=Multiply(A=@n81, B=LinearInterpolate(A=@n102, B=@n103, Alpha=@n101.AddMultiply)))
  n105 = Add(
   A=Multiply(A=Multiply(A=@n56.RGB, B=@n55.PlaneColorPower), B=Multiply(A=@n81, B=@n87))
   B=Multiply(A=Multiply(A=@n62, B=@n55.NoiseInnerColor), B=Multiply(A=@n104, B=@n101.Brightness))
  )
  MaterialFunctionCall(Emissive=Add(A=Add(A=@n35, B=@n54), B=Add(A=@n86, B=@n105))) {materialFunction=&MF_FX_EmissiveAdjustments}
  n106 = TextureCoordinate()
  n107 = Power(Base=OneMinus(None=ComponentMask(None=@n106) {r=true}), Exp=Constant() {r=-5})
  n108 = Multiply(A=Multiply(A=Constant() {r=60}, B=Multiply(A=@n107, B=@n24)))
  n109 = Saturate(None=Multiply(A=Constant() {r=80}, B=Saturate(None=@n108)))
  n110 = TextureCoordinate()
  n111 = Power(Base=OneMinus(None=ComponentMask(None=@n110) {r=true}), Exp=Constant() {r=-5})
  n112 = Multiply(A=Multiply(A=Constant() {r=60}, B=Multiply(A=@n111, B=@n75)))
  n113 = Saturate(None=Multiply(A=Constant() {r=0.85}, B=Saturate(None=@n112)))
  n114 = Sine(None=ComponentMask(None=TextureCoordinate()) {g=true}) {period=2}
  Multiply(
   A=Multiply(A=Power(Base=@n109) {constExponent=0.1}, B=Power(Base=@n113) {constExponent=0.1})
   B=Round(None=Power(Base=@n114) {constExponent=1})
  )
 ]
}
